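# Control Rig for the GrimeHold male medium body cosmetic, adding corrective deformation and driving its neck and belt accessory bones.
# A sequence runs the shared DeformRig function, weighted parent constraints, and a custom spherical collider that clamps neck bones within radii around the head.
# Spherical pose readers on head, clavicles (parachute only) and left thigh weight corrective offsets.
# Pelvis height is remapped through a curve into CrouchWalk.
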ControlRigBlueprint /BRCosmetics/Characters/Player/Male/Medium/Bodies/M_MED_GrimeHold/Meshes/M_MED_GrimeHold_CtrlRig
graphs [
 New Function_ContainedGraph {nodes [
   n1 = RigVMFunctionEntryNode()
   n2 = &RigUnit_GetTransform(bInitial=False, Item=@n1.Source)
   ClampSpatially = /Script/RigVM.RigVMFunction_MathTransformClampSpatially(
    Axis=None
    DebugColor=(R=1,G=0,B=0,A=1)
    DebugThickness=1
    Type=RigVMDispatchNode(Value=@n1.Type) {template="DISPATCH_RigVMDispatch_CastIntToEnum(in Value,out Result)"}.Result
    bDrawDebug=@n1.Debug
    Maximum=@n1.Max Radius
    Minimum=@n1.Min Radius
    Space=&RigVMFunction_MathTransformMul(
     B=&RigUnit_GetTransform(bInitial=False, Item=@n1.Target).Transform
     A=/Script/RigVM.RigVMFunction_MathTransformMake(Rotation=(X=0,Y=0,Z=0,W=1), Scale=(X=1,Y=1,Z=1), Translation=@n1.Target Offset) {is=struct RigVMUnitNode}.Result
    ) {template="Multiply::Execute(in A,in B,out Result)"}.Result
    Value=&RigVMFunction_MathTransformMul(
     B=@n2.Transform
     A=&RigVMFunction_MathTransformMake(Rotation=(X=0,Y=0,Z=0,W=1), Scale=(X=1,Y=1,Z=1), Translation=@n1.Source Offset).Result
    ) {template="Multiply::Execute(in A,in B,out Result)"}.Result
   ) {
    is=struct RigVMUnitNode
    template="ClampSpatially::Execute(in Value,in Axis,in Type,in Minimum,in Maximum,in Space,in bDrawDebug,in DebugColor,in DebugThickness,out Result)"
   }
   GreaterEqual = /Script/RigVM.RigVMFunction_MathDoubleGreaterEqual(B=10.3, A=&RigVMFunction_MathVectorDistance(B=@ClampSpatially.Result.Translation, A=@n2.Transform.Translation).Result) {is=struct RigVMUnitNode, template="GreaterEqual::Execute(in A,in B,out Result)"}
   n3 = &RigVMFunction_ControlFlowBranch(ExecuteContext=@n1.ExecuteContext, Condition=@GreaterEqual.Result)
   RigVMFunctionReturnNode(IsColliding=false, ExecuteContext=@n3.Completed)
   &RigUnit_ModifyTransforms(
    ItemToModify=((Item=(Type=Bone,Name="None"),Transform=(Rotation=(X=0,Y=0,Z=0,W=1),Translation=(X=0,Y=0,Z=0),Scale3D=(X=1,Y=1,Z=1))))
    Mode=OverrideGlobal
    WeightMinimum=0
    ItemToModify.0.Item=@n1.Source
    ExecutePin=@n3.True
    Weight=@n1.Weight
    ItemToModify.0.Transform=@ClampSpatially.Result
   )
   LessEqual = /Script/RigVM.RigVMFunction_MathDoubleLessEqual(A=0, B=10) {is=struct RigVMUnitNode, template="LessEqual::Execute(in A,in B,out Result)"}
  ]}
 RigVMFunctionLibrary {nodes [Spherical Collider v02 = RigVMCollapseNode(ExecuteContext=(), Source=(Type=None,Name="None"), Source Offset=(X=0,Y=0,Z=0), Target=(Type=None,Name="None"), Target Offset=(X=0,Y=0,Z=0))]}
 RigVMFunction_Sequence_1_ContainedGraph {nodes [
   n1 = RigVMFunctionEntryNode()
   RigVMFunction_Sequence = &RigVMFunction_Sequence(ExecuteContext=@n1.ExecuteContext)
   n2 = &RigVMFunction_Sequence(ExecuteContext=@RigVMFunction_Sequence.B)
   n3 = &RigVMFunction_Sequence(ExecuteContext=@n2.B)
   n4 = &RigVMFunction_Sequence(ExecuteContext=@n3.B)
   n5 = &RigVMFunction_Sequence(ExecuteContext=@n4.B)
   n6 = &RigVMFunction_Sequence(ExecuteContext=@n5.B)
   n7 = &RigVMFunction_Sequence(ExecuteContext=@n6.B)
   RigVMFunctionReturnNode(A=@RigVMFunction_Sequence.A, B=@n2.A, C=@n3.A, D=@n4.A, E=@n5.A, F=@n6.A, G=@n7.A, H=@n7.B)
  ]}
 RigVMModel {nodes [
   RigUnit_BeginExecution = &RigUnit_BeginExecution()
   RigVMFunction_Sequence = RigVMAggregateNode(ExecuteContext=@RigUnit_BeginExecution.ExecutePin)
   FN_DeformRig_FNC = RigVMFunctionReferenceNode(
    DeformRig_ Curve="(deformRig_l_upperarm_up_alpha_72_#1=False,deformRig_r_upperarm_up_alpha_73_#2=False,deformRig_l_upperarm_fr_alpha_74_#3=False,deformRig_r_upperarm_fr_alpha_75_CB51B8 ...(truncated)"
    DeformRig_Bool="(upperarm_1_#4=True,elbow_3_#5=True,wrist_7_#6=True,groin_8_#7=True,glute_17_#8=True,knee_9_40C408324CA261 ...(truncated)"
    DeformRig_Debug=(upperarm_debug_18_#4=False,wrist_debug_20_#6=False,lat_debug_26_#9=False,groinGlute_debug_27_#7=False)
    DeformRig_Value="(upperarm_up_18_#10=(Rotation=(X=0,Y=0,Z=0,W=1),Translation=(X=0,Y=0,Z=8),Scale3D=(X=0.7,Y=1,Z=1)),upperarm_fwd_fr_19_#11=(Rotation=(X=0,Y=0,Z=0,W=1),Translation=(X=0,Y=10,Z=0),Scale3D=(X=0.5,Y=1,Z= ...(truncated)"
    ExecuteContext=@RigVMFunction_Sequence.B
   )
   &RigUnit_ParentConstraint(
    AdvancedSettings=(RotationOrderForFilter=XZY)
    Child=(Type=Bone,Name="dyn_neck_acc_r")
    Parents="((Item=(Type=Bone,Name=\"spine_05\"),Weight=0.0002),(Item=(Type=Bone,Name=\"dyn_neck_03_up_r\"),Weight=0.0036),(Item=(Type=Bone,Name=\"dyn_neck_03_dw_r\"),Weight=0.0045),(Item=(Type=Bone,Name=\"spine_04\"),Weight=0.0179),(Item=(Type=Bone,Name=\"dyn_neck_fr\"),Weight ...(truncated)"
    Weight=1
    ExecutePin=@RigVMFunction_Sequence.C
   )
   > &RigUnit_ParentConstraint(
    AdvancedSettings=(RotationOrderForFilter=XZY)
    Child=(Type=Bone,Name="dyn_neck_acc_l")
    Parents="((Item=(Type=Bone,Name=\"spine_05\"),Weight=0.0004),(Item=(Type=Bone,Name=\"dyn_neck_03_up_l\"),Weight=0.0039),(Item=(Type=Bone,Name=\"dyn_neck_03_dw_l\"),Weight=0.0125),(Item=(Type=Bone,Name=\"dyn_neck_fr\"),Weight=0.013),(Item=(Type=Bone,Name=\"spine_04\"),Weight= ...(truncated)"
    Weight=1
   )
   > Spherical Collision_1 = RigVMFunctionReferenceNode(
    Debug=false
    Max Radius=35
    Min Radius=22
    Source=(Type=Bone,Name="dyn_neck_fr")
    Source Offset=(X=0,Y=0,Z=0)
    Target=(Type=Bone,Name="head")
    Target Offset=(X=3,Y=2.5,Z=0)
    Weight=1
   )
   > Spherical Collision_3 = RigVMFunctionReferenceNode(
    Debug=false
    Max Radius=24
    Min Radius=19
    Source=(Type=Bone,Name="dyn_neck_01_l")
    Source Offset=(X=0,Y=0,Z=0)
    Target=(Type=Bone,Name="head")
    Target Offset=(X=2,Y=4,Z=0)
    Weight=1
   )
   > Spherical Collision_4 = RigVMFunctionReferenceNode(
    Debug=false
    Max Radius=24
    Min Radius=19
    Source=(Type=Bone,Name="dyn_neck_02_up_l")
    Source Offset=(X=0,Y=0,Z=0)
    Target=(Type=Bone,Name="head")
    Target Offset=(X=2,Y=4,Z=0)
    Weight=1
   )
   > Spherical Collision_5 = RigVMFunctionReferenceNode(
    Debug=false
    Max Radius=24
    Min Radius=19
    Source=(Type=Bone,Name="dyn_neck_03_up_l")
    Source Offset=(X=0,Y=0,Z=0)
    Target=(Type=Bone,Name="head")
    Target Offset=(X=2,Y=4,Z=0)
    Weight=1
   )
   > Spherical Collision_0 = RigVMFunctionReferenceNode(
    Debug=false
    Max Radius=30
    Min Radius=21.7
    Source=(Type=Bone,Name="dyn_neck_bck_up_l")
    Source Offset=(X=0,Y=0,Z=0)
    Target=(Type=Bone,Name="head")
    Target Offset=(X=2,Y=5.5,Z=0)
    Weight=1
   )
   n1 = /Script/RigVM.RigVMFunction_MathBoolMake(Value=false) {is=struct RigVMUnitNode}
   > Spherical Collision_6 = RigVMFunctionReferenceNode(
    Max Radius=30
    Min Radius=15
    Source=(Type=Bone,Name="dyn_neck_01_r")
    Source Offset=(X=0,Y=0,Z=0)
    Target=(Type=Bone,Name="head")
    Target Offset=(X=-10,Y=-0.5,Z=1)
    Weight=1
    Debug=@n1.Value
   )
   > Spherical Collision_7 = RigVMFunctionReferenceNode(
    Max Radius=35
    Min Radius=18
    Source=(Type=Bone,Name="dyn_neck_02_up_r")
    Source Offset=(X=0,Y=0,Z=0)
    Target=(Type=Bone,Name="head")
    Target Offset=(X=-4,Y=0,Z=-1)
    Weight=1
    Debug=@n1.Value
   )
   > Spherical Collision_8 = RigVMFunctionReferenceNode(
    Max Radius=30
    Min Radius=14.8
    Source=(Type=Bone,Name="dyn_neck_03_up_r")
    Source Offset=(X=0,Y=0,Z=0)
    Target=(Type=Bone,Name="head")
    Target Offset=(X=1,Y=0,Z=0)
    Weight=1
    Debug=@n1.Value
   )
   n2 = &RigUnit_SphericalPoseReader(
    ActiveRegionScaleFactors=(PositiveWidth=1,NegativeWidth=1,PositiveHeight=1,NegativeHeight=1)
    ActiveRegionSize=0.1
    Debug=(bDrawDebug=false,DebugScale=25,DebugSegments=20,DebugThickness=0.25)
    DriverAxis=(X=0,Y=-20,Z=0)
    DriverItem=(Type=Bone,Name="head")
    FalloffRegionScaleFactors=(PositiveWidth=0.5,NegativeWidth=0.5,PositiveHeight=0.5,NegativeHeight=0.5)
    FalloffSize=0.5
    OptionalParentItem=(Type=Bone)
    RotationOffset=(X=-90,Y=0,Z=-80)
    ExecutePin=@RigVMFunction_Sequence.D
   )
   > &RigUnit_ModifyTransforms(
    ItemToModify="((Item=(Type=Bone,Name=\"dyn_neck_bck_up_l\"),Transform=(Rotation=(X=0,Y=0,Z=0,W=-1),Translation=(X=1,Y=0,Z=-3),Scale3D=(X=1,Y=1,Z=1))),(Item=(Type=Bone,Name=\"dyn_neck_bck_center_l\"),Transform=(Rotation=(X=0,Y=0,Z=0,W=1),Translation=(X=3.5,Y=0,Z=-2.7),Scale3 ...(truncated)"
    WeightMinimum=0
    Weight=@n2.OutputParam
   )
   > n3 = &RigUnit_SphericalPoseReader(
    ActiveRegionScaleFactors=(PositiveWidth=1,NegativeWidth=1,PositiveHeight=1,NegativeHeight=1)
    ActiveRegionSize=0.1
    Debug=(bDrawDebug=false,DebugScale=25,DebugSegments=20,DebugThickness=0.25)
    DriverAxis=(X=0,Y=1.2,Z=1)
    DriverItem=(Type=Bone,Name="head")
    FalloffRegionScaleFactors=(PositiveWidth=0.5,NegativeWidth=0.5,PositiveHeight=0.5,NegativeHeight=0.5)
    FalloffSize=0.5
    OptionalParentItem=(Type=Bone)
    RotationOffset=(X=-30,Y=-10,Z=0)
   )
   > &RigUnit_ModifyTransforms(
    ItemToModify="((Item=(Type=Bone,Name=\"dyn_neck_03_up_r\"),Transform=(Rotation=(X=0,Y=0,Z=0,W=1),Translation=(X=3,Y=1,Z=-3.7),Scale3D=(X=1,Y=1,Z=1))),(Item=(Type=Bone,Name=\"dyn_neck_02_up_r\"),Transform=(Rotation=(X=0,Y=0,Z=0,W=1),Translation=(X=2.7,Y=4.3,Z=-5),Scale3D=(X= ...(truncated)"
    WeightMinimum=0
    Weight=@n3.OutputParam
   )
   > n4 = &RigUnit_SphericalPoseReader(
    ActiveRegionScaleFactors=(PositiveWidth=1,NegativeWidth=1,PositiveHeight=1,NegativeHeight=1)
    ActiveRegionSize=0.1
    Debug=(bDrawDebug=false,DebugScale=25,DebugSegments=20,DebugThickness=0.25)
    DriverAxis=(X=0,Y=0,Z=-1)
    DriverItem=(Type=Bone,Name="head")
    FalloffRegionScaleFactors=(PositiveWidth=0.5,NegativeWidth=0.5,PositiveHeight=0.5,NegativeHeight=0.5)
    FalloffSize=0.5
    OptionalParentItem=(Type=Bone)
    RotationOffset=(X=235,Y=-10,Z=0)
   )
   > &RigUnit_ModifyTransforms(
    ItemToModify="((Item=(Type=Bone,Name=\"dyn_neck_03_up_l\"),Transform=(Rotation=(X=0,Y=0,Z=0,W=1),Translation=(X=-1.2,Y=0.5,Z=-3.5),Scale3D=(X=1,Y=1,Z=1))),(Item=(Type=Bone,Name=\"dyn_neck_02_up_l\"),Transform=(Rotation=(X=0,Y=0,Z=0,W=1),Translation=(X=1.2,Y=-4.2,Z=1),Scale3 ...(truncated)"
    WeightMinimum=0
    Weight=@n4.OutputParam
   )
   RigVMFunction_ControlFlowBranch = &RigVMFunction_ControlFlowBranch(ExecuteContext=@RigVMFunction_Sequence.E, Condition=RigVMVariableNode() {variable=IsParachute}.Value)
   n5 = &RigVMFunction_MathBoolMake(Value=false)
   n6 = &RigUnit_SphericalPoseReader(
    ActiveRegionScaleFactors=(PositiveWidth=1,NegativeWidth=1,PositiveHeight=1,NegativeHeight=1)
    ActiveRegionSize=0.1
    Debug=(bDrawDebug=false,DebugScale=25,DebugSegments=20,DebugThickness=0.25)
    DriverAxis=(X=1,Y=0,Z=0)
    DriverItem=(Type=Bone,Name="clavicle_l")
    FalloffRegionScaleFactors=(PositiveWidth=0.5,NegativeWidth=0.5,PositiveHeight=0.5,NegativeHeight=0.5)
    FalloffSize=0.5
    OptionalParentItem=(Type=Bone)
    RotationOffset=(X=0,Y=-40,Z=0)
    ExecutePin=@RigVMFunction_ControlFlowBranch.True
    Debug.bDrawDebug=@n5.Value
   )
   > &RigUnit_ModifyTransforms(
    ItemToModify="((Item=(Type=Bone,Name=\"dyn_neck_02_up_l\"),Transform=(Rotation=(X=0,Y=0,Z=0,W=1),Translation=(X=-3.7,Y=0,Z=0),Scale3D=(X=1,Y=1,Z=1))),(Item=(Type=Bone,Name=\"dyn_neck_02_dw_l\"),Transform=(Rotation=(X=0,Y=0,Z=0,W=1),Translation=(X=-1,Y=0,Z=0),Scale3D=(X=1,Y= ...(truncated)"
    WeightMinimum=0
    Weight=@n6.OutputParam
   )
   > n7 = &RigUnit_SphericalPoseReader(
    ActiveRegionScaleFactors=(PositiveWidth=1,NegativeWidth=1,PositiveHeight=1,NegativeHeight=1)
    ActiveRegionSize=0.1
    Debug=(bDrawDebug=false,DebugScale=25,DebugSegments=20,DebugThickness=0.25)
    DriverAxis=(X=-1,Y=0,Z=0)
    DriverItem=(Type=Bone,Name="clavicle_r")
    FalloffRegionScaleFactors=(PositiveWidth=0.5,NegativeWidth=0.5,PositiveHeight=0.5,NegativeHeight=0.5)
    FalloffSize=0.5
    OptionalParentItem=(Type=Bone)
    RotationOffset=(X=180,Y=-40,Z=0)
    Debug.bDrawDebug=@n5.Value
   )
   > &RigUnit_ModifyTransforms(
    ItemToModify="((Item=(Type=Bone,Name=\"dyn_neck_02_up_r\"),Transform=(Rotation=(X=0,Y=0,Z=0,W=1),Translation=(X=-3.7,Y=0,Z=0),Scale3D=(X=1,Y=1,Z=1))),(Item=(Type=Bone,Name=\"dyn_neck_02_dw_r\"),Transform=(Rotation=(X=0,Y=0,Z=0,W=1),Translation=(X=-1,Y=0,Z=0),Scale3D=(X=1,Y= ...(truncated)"
    WeightMinimum=0
    Weight=@n7.OutputParam
   )
   n8 = &RigVMFunction_AnimEvalRichCurve(
    Curve=(EditorCurveData=(Keys=((),(Time=0.5,Value=1),(Time=0.8))))
    SourceMaximum=2
    SourceMinimum=0
    TargetMinimum=0
    Value=&RigVMFunction_MathDoubleRemap(
     SourceMaximum=40
     SourceMinimum=90
     TargetMaximum=1
     TargetMinimum=0
     bClamp=false
     Value=&RigUnit_GetRelativeTransformForItem(Child=(Type=Bone,Name="pelvis"), Parent=(Type=Bone,Name="root"), bChildInitial=False, bParentInitial=False).RelativeTransform.Translation.Z
    ).Result
   )
   RigVMVariableNode(Value=@n8.Result, ExecuteContext=@RigVMFunction_Sequence.F) {variable=CrouchWalk}
   &RigUnit_ParentConstraint(
    AdvancedSettings=(RotationOrderForFilter=XZY)
    Child=(Type=Bone,Name="base_acc_02_l")
    Parents="((Item=(Type=Bone,Name=\"deform_groin_r\"),Weight=0.003),(Item=(Type=Bone,Name=\"thigh_l\"),Weight=0.0063),(Item=(Type=Bone,Name=\"thigh_r\"),Weight=0.0072),(Item=(Type=Bone,Name=\"spine_04\"),Weight=0.0092),(Item=(Type=Bone,Name=\"spine_03\"),Weight=0.0775),(Item=( ...(truncated)"
    Weight=1
    ExecutePin=@RigVMFunction_Sequence.G
   )
   > &RigUnit_ParentConstraint(
    AdvancedSettings=(RotationOrderForFilter=XZY)
    Child=(Type=Bone,Name="base_acc_01_l")
    Parents="((Item=(Type=Bone,Name=\"deform_groin_r\"),Weight=0.0031),(Item=(Type=Bone,Name=\"spine_04\"),Weight=0.0089),(Item=(Type=Bone,Name=\"thigh_r\"),Weight=0.0109),(Item=(Type=Bone,Name=\"thigh_l\"),Weight=0.0121),(Item=(Type=Bone,Name=\"spine_03\"),Weight=0.0758),(Item= ...(truncated)"
    Weight=1
   )
   > &RigUnit_ParentConstraint(
    AdvancedSettings=(RotationOrderForFilter=XZY)
    Child=(Type=Bone,Name="base_acc_02_r")
    Parents="((Item=(Type=Bone,Name=\"deform_groin_r\"),Weight=0.003),(Item=(Type=Bone,Name=\"thigh_l\"),Weight=0.0063),(Item=(Type=Bone,Name=\"thigh_r\"),Weight=0.0072),(Item=(Type=Bone,Name=\"spine_04\"),Weight=0.0099),(Item=(Type=Bone,Name=\"spine_03\"),Weight=0.0786),(Item=( ...(truncated)"
    Weight=1
   )
   > &RigUnit_ParentConstraint(
    AdvancedSettings=(RotationOrderForFilter=XZY)
    Child=(Type=Bone,Name="base_acc_01_r")
    Parents="((Item=(Type=Bone,Name=\"deform_groin_r\"),Weight=0.0051),(Item=(Type=Bone,Name=\"thigh_l\"),Weight=0.0059),(Item=(Type=Bone,Name=\"thigh_r\"),Weight=0.0074),(Item=(Type=Bone,Name=\"spine_04\"),Weight=0.009),(Item=(Type=Bone,Name=\"spine_03\"),Weight=0.0743),(Item=( ...(truncated)"
    Weight=1
   )
   > n9 = &RigUnit_SphericalPoseReader(
    ActiveRegionScaleFactors=(PositiveWidth=1,NegativeWidth=1,PositiveHeight=1,NegativeHeight=1)
    ActiveRegionSize=0.05
    Debug=(bDrawDebug=false,bDraw2D=False,bDrawLocalAxes=False,DebugScale=25,DebugSegments=20,DebugThickness=0.25)
    DriverAxis=(X=-1,Y=0,Z=0)
    DriverItem=(Type=Bone,Name="thigh_l")
    FalloffRegionScaleFactors=(PositiveWidth=1,NegativeWidth=1,PositiveHeight=0.9,NegativeHeight=0.9)
    FalloffSize=0.4
    FlipHeightScaling=False
    FlipWidthScaling=False
    OptionalParentItem=(Type=Bone,Name="None")
    RotationOffset=(X=90,Y=0,Z=-30)
   )
   > &RigUnit_ModifyTransforms(
    ItemToModify=((Item=(Type=Bone,Name="base_acc_01_l"),Transform=(Rotation=(X=0.036,Y=0.0062,Z=0.1206,W=0.992),Translation=(X=-1.2048,Y=1.1258,Z=0.687),Scale3D=(X=1,Y=1,Z=1))))
    WeightMinimum=0
    Weight=@n9.OutputParam
   )
   > n10 = &RigUnit_SphericalPoseReader(
    ActiveRegionScaleFactors=(PositiveWidth=1,NegativeWidth=1,PositiveHeight=1,NegativeHeight=1)
    ActiveRegionSize=0.05
    Debug=(bDrawDebug=false,bDraw2D=False,bDrawLocalAxes=False,DebugScale=25,DebugSegments=20,DebugThickness=0.25)
    DriverAxis=(X=-1,Y=0,Z=0)
    DriverItem=(Type=Bone,Name="thigh_l")
    FalloffRegionScaleFactors=(PositiveWidth=1,NegativeWidth=1,PositiveHeight=0.9,NegativeHeight=0.9)
    FalloffSize=0.4
    FlipHeightScaling=False
    FlipWidthScaling=False
    OptionalParentItem=(Type=Bone,Name="None")
    RotationOffset=(X=140,Y=30,Z=0)
   )
   > &RigUnit_ModifyTransforms(
    ItemToModify="((Item=(Type=Bone,Name=\"base_acc_01_l\"),Transform=(Rotation=(X=0.4384,Y=-0.2827,Z=-0.0839,W=0.8491),Translation=(X=0.0342,Y=-6.902,Z=0.9644),Scale3D=(X=1,Y=1,Z=1))),(Item=(Type=Bone,Name=\"base_acc_02_l\"),Transform=(Rotation=(X=0.2624,Y=-0.0722,Z=0.0222,W=0 ...(truncated)"
    WeightMinimum=0
    Weight=@n10.OutputParam
   )
   &RigUnit_SetTransform(
    ExecutePin=()
    Item=(Type=Bone,Name="neck_01")
    Weight=1
    bInitial=False
    Value=&RigUnit_GetTransform(Item=(Type=Control,Name="neck_01_ctrl"), bInitial=False).Transform
   )
   > &RigUnit_SetTransform(
    Item=(Type=Bone,Name="neck_02")
    Weight=1
    bInitial=False
    Value=&RigUnit_GetTransform(Item=(Type=Control,Name="neck_02_ctrl"), bInitial=False).Transform
   )
   RigUnit_GetTransform = &RigUnit_GetTransform(Item=(Type=Control,Name="head_ctrl"), bInitial=False)
   > RigUnit_SetTransform = &RigUnit_SetTransform(Item=(Type=Bone,Name="head"), Weight=1, bInitial=False, Value=@RigUnit_GetTransform.Transform)
   &RigUnit_SetTransform(
    ExecutePin=()
    Item=(Type=Bone,Name="clavicle_l")
    Weight=1
    bInitial=False
    Value=&RigUnit_GetTransform(Item=(Type=Control,Name="clavicle_l_ctrl"), bInitial=False).Transform
   )
   Spherical Collision_2 = RigVMFunctionReferenceNode(
    Debug=false
    ExecuteContext=()
    Max Radius=25
    Min Radius=16
    Source=(Type=Bone,Name="dyn_neck_bck_center_l")
    Source Offset=(X=0,Y=0,Z=0)
    Target=(Type=Bone,Name="head")
    Target Offset=(X=-6,Y=-2,Z=0)
    Weight=1
   )
  ]}
]
hierarchy [
 root {loc=<0 0 0>, rot=<0 0 0 1>}
 pelvis {parent=root, loc=<0 -0.7071 93.1191>, rot=<0.0127 0.707 -0.0127 -0.707>}
 spine_01 {parent=pelvis, loc=<6.7494 -0.6388 0>, rot=<0 0 0.034 -0.9994>}
 spine_02 {parent=spine_01, loc=<7.2518 -0.035 0>, rot=<0 0 0.0238 -0.9997>}
 spine_03 {parent=spine_02, loc=<6.0104 0 0>, rot=<0 0 0.0149 -0.9999>}
 spine_04 {parent=spine_03, loc=<8.1024 0 0>, rot=<0 0 -0.0132 -0.9999>}
 spine_05 {parent=spine_04, loc=<10.0146 0 0>, rot=<0 0 -0.0015 -1>}
 neck_01 {parent=spine_05, loc=<11.3138 0.1529 0>, rot=<0 0 -0.2003 -0.9797>}
 neck_02 {parent=neck_01, loc=<4.5627 0 0>, rot=<0 0 0.0504 -0.9987>}
 head {parent=neck_02, loc=<5.3914 0 0>, rot=<0 0 0.115 -0.9934>}
 attach_cape {parent=spine_05, loc=<9.017 -11.714 0>, rot=<0.0616 -0.7044 -0.0616 -0.7044>}
 cape {parent=attach_cape, type=Null, loc=<0 -5.7715 0>, rot=<0 0 0 1>}
 attach_backpack {parent=spine_05, loc=<-0.123 -13.473 0>, rot=<-0.0093 -0.707 0.0093 -0.707>}
 Backpack_BR {parent=attach_backpack, type=Null, loc=<0 -3.5926 0>, rot=<0 0 0 1>}
 clavicle_l {parent=spine_05, loc=<4.3575 -0.1466 -2.3223>, rot=<0.0465 -0.7289 -0.009 -0.6829>}
 upperarm_l {parent=clavicle_l, loc=<15.5111 0 0>, rot=<0.0226 -0.3234 -0.0088 -0.9459>}
 lowerarm_l {parent=upperarm_l, loc=<26.6514 0 0>, rot=<0.0391 -0.0099 -0.2453 -0.9686>}
 hand_l {parent=lowerarm_l, loc=<22.3491 0 0>, rot=<-0.7081 -0.0507 0.0975 -0.6975>}
 weapon_l {parent=hand_l, loc=<-0.1974 -3.4111 0.3811>, rot=<0 0 0 1>}
 thumb_01_l {parent=hand_l, loc=<1.3118 -0.3191 -2.459>, rot=<0.6546 -0.4042 0.0664 -0.6354>}
 thumb_02_l {parent=thumb_01_l, loc=<5.9336 -0.0448 0.0401>, rot=<-0.0037 0.0213 0.298 -0.9543>}
 thumb_03_l {parent=thumb_02_l, loc=<2.2223 0.3682 -0.0907>, rot=<0.0118 -0.013 0.0066 -0.9998>}
 index_metacarpal_l {parent=hand_l, loc=<1.6535 0.0183 -2.96>, rot=<0.0158 -0.0124 -0.0173 -0.9996>}
 index_01_l {parent=index_metacarpal_l, loc=<8.3008 0.0357 -0.0524>, rot=<-0.0044 0.0485 0.0924 -0.9945>}
 index_02_l {parent=index_01_l, loc=<3.3587 -0.3004 0.0452>, rot=<-0.002 0.0027 0.3225 -0.9466>}
 index_03_l {parent=index_02_l, loc=<2.5578 -0.123 -0.0252>, rot=<-0.002 0.0067 0.2558 -0.9667>}
 middle_metacarpal_l {parent=hand_l, loc=<2.1856 -0.1725 -0.5984>, rot=<0.006 0.0168 -0.0329 -0.9993>}
 middle_01_l {parent=middle_metacarpal_l, loc=<7.3275 -0.2447 0.0129>, rot=<-0.0208 0.0042 0.0858 -0.9961>}
 middle_02_l {parent=middle_01_l, loc=<4.0123 0.1401 0.1175>, rot=<-0.0218 0.0258 0.3512 -0.9357>}
 middle_03_l {parent=middle_02_l, loc=<2.6077 0.0534 0.1373>, rot=<0.0082 0.0188 0.1708 -0.9851>}
 ring_metacarpal_l {parent=hand_l, loc=<2.2074 -0.3625 1.3638>, rot=<-0.0284 0.0722 -0.0458 -0.9959>}
 ring_01_l {parent=ring_metacarpal_l, loc=<7.0295 -0.6197 -0.1446>, rot=<-0.0489 -0.0252 0.1892 -0.9804>}
 ring_02_l {parent=ring_01_l, loc=<3.6174 0.4647 -0.1027>, rot=<0 0 0.2806 -0.9598>}
 ring_03_l {parent=ring_02_l, loc=<2.4152 0.114 0.0199>, rot=<0 0 0.1189 -0.9929>}
 pinky_metacarpal_l {parent=hand_l, loc=<1.7616 -0.9753 3.3814>, rot=<-0.0397 0.0814 -0.069 -0.9935>}
 pinky_01_l {parent=pinky_metacarpal_l, loc=<7.1709 -0.4831 -0.1094>, rot=<-0.0654 -0.0035 0.2242 -0.9723>}
 pinky_02_l {parent=pinky_01_l, loc=<2.8274 0.0893 -0.1043>, rot=<0.0037 0.0011 0.2854 -0.9584>}
 pinky_03_l {parent=pinky_02_l, loc=<2.2046 0.0307 -0.0078>, rot=<-0.0002 0.0098 0.1445 -0.9895>}
 hand_attach_l {parent=hand_l, loc=<8.5 -3 0>, rot=<0.7071 0.7071 0 0>}
 deform_wrist_dn_base_l {parent=hand_l, loc=<0 -2.2342 0>, rot=<0 0 0 -1>}
 deform_wrist_dn_l {parent=deform_wrist_dn_base_l, loc=<0 0 0>, rot=<0 0 0 1>}
 deform_wrist_up_base_l {parent=hand_l, loc=<0 2.2682 0>, rot=<0 0 0 -1>}
 deform_wrist_up_l {parent=deform_wrist_up_base_l, loc=<0 0 0>, rot=<0 0 0 1>}
 lowerarm_twist_02_l {parent=lowerarm_l, loc=<9.6199 0 0>, rot=<0 0 0 -1>}
 base_gloves_twist_l {parent=lowerarm_twist_02_l, loc=<0 0 0>, rot=<0 0 0 1>}
 dyn_gloves_twist_l {parent=base_gloves_twist_l, loc=<0 0 0>, rot=<0 0 0 1>}
 lowerarm_twist_01_l {parent=lowerarm_l, loc=<17.8549 0 0>, rot=<0 0 0 -1>}
 deform_elbow_in_base_l {parent=lowerarm_l, loc=<0.2112 5.259 0.8274>, rot=<0 0 0 -1>}
 deform_elbow_in_l {parent=deform_elbow_in_base_l, loc=<0 0 0>, rot=<0 0 0 1>}
 deform_elbow_out_base_l {parent=lowerarm_l, loc=<0 -4 0>, rot=<0 0 0 -1>}
 deform_elbow_out_l {parent=deform_elbow_out_base_l, loc=<0 0 0>, rot=<0 0 0 1>}
 upperarm_twist_01_l {parent=upperarm_l, loc=<0.0041 0 0>, rot=<0 0 0 1>}
 deform_upperarm_bk_base_l {parent=upperarm_twist_01_l, loc=<1.0813 -7.2193 -1.9453>, rot=<0 0 0 1>}
 deform_upperarm_bk_l {parent=deform_upperarm_bk_base_l, loc=<0 0 0>, rot=<0 0 0 1>}
 deform_upperarm_fr_base_l {parent=upperarm_twist_01_l, loc=<-0.3691 7.0506 -2.765>, rot=<0 0 0 1>}
 deform_upperarm_fr_l {parent=deform_upperarm_fr_base_l, loc=<0 0 0>, rot=<0 0 0 1>}
 deform_upperarm_up_base_l {parent=upperarm_twist_01_l, loc=<-3.2216 -1.1391 3.3311>, rot=<0 0 0 1>}
 deform_upperarm_up_l {parent=deform_upperarm_up_base_l, loc=<0 0 0>, rot=<0 0 0 1>}
 upperarm_twist_02_l {parent=upperarm_l, loc=<13.7941 0 0>, rot=<0 0 0 1>}
 clavicle_r {parent=spine_05, loc=<4.3575 -0.1466 2.3223>, rot=<0.7289 0.0465 -0.6829 0.009>}
 upperarm_r {parent=clavicle_r, loc=<-15.5111 0 0>, rot=<0.0226 -0.3234 -0.0088 -0.9459>}
 lowerarm_r {parent=upperarm_r, loc=<-26.6514 0 0>, rot=<0.0391 -0.0099 -0.2453 -0.9686>}
 hand_r {parent=lowerarm_r, loc=<-22.3491 0 0>, rot=<-0.7081 -0.0507 0.0975 -0.6975>}
 weapon_r {parent=hand_r, loc=<0.1974 3.4111 -0.3811>, rot=<0 0 0 1>}
 thumb_01_r {parent=hand_r, loc=<-1.3118 0.3191 2.459>, rot=<0.6546 -0.4042 0.0664 -0.6354>}
 thumb_02_r {parent=thumb_01_r, loc=<-5.9336 0.0448 -0.0401>, rot=<-0.0037 0.0213 0.298 -0.9543>}
 thumb_03_r {parent=thumb_02_r, loc=<-2.2223 -0.3682 0.0907>, rot=<0.0118 -0.013 0.0066 -0.9998>}
 index_metacarpal_r {parent=hand_r, loc=<-1.6535 -0.0183 2.96>, rot=<0.0158 -0.0124 -0.0173 -0.9996>}
 index_01_r {parent=index_metacarpal_r, loc=<-8.3008 -0.0357 0.0524>, rot=<-0.0044 0.0485 0.0924 -0.9945>}
 index_02_r {parent=index_01_r, loc=<-3.3587 0.3004 -0.0452>, rot=<-0.002 0.0027 0.3225 -0.9466>}
 index_03_r {parent=index_02_r, loc=<-2.5578 0.123 0.0252>, rot=<-0.002 0.0067 0.2558 -0.9667>}
 middle_metacarpal_r {parent=hand_r, loc=<-2.1856 0.1725 0.5984>, rot=<0.006 0.0168 -0.0329 -0.9993>}
 middle_01_r {parent=middle_metacarpal_r, loc=<-7.3275 0.2447 -0.0129>, rot=<-0.0208 0.0042 0.0858 -0.9961>}
 middle_02_r {parent=middle_01_r, loc=<-4.0123 -0.1401 -0.1175>, rot=<-0.0218 0.0258 0.3512 -0.9357>}
 middle_03_r {parent=middle_02_r, loc=<-2.6077 -0.0534 -0.1373>, rot=<0.0082 0.0188 0.1708 -0.9851>}
 ring_metacarpal_r {parent=hand_r, loc=<-2.2074 0.3625 -1.3638>, rot=<-0.0284 0.0722 -0.0458 -0.9959>}
 ring_01_r {parent=ring_metacarpal_r, loc=<-7.0295 0.6197 0.1446>, rot=<-0.0489 -0.0252 0.1892 -0.9804>}
 ring_02_r {parent=ring_01_r, loc=<-3.6174 -0.4647 0.1027>, rot=<0 0 0.2806 -0.9598>}
 ring_03_r {parent=ring_02_r, loc=<-2.4152 -0.114 -0.0199>, rot=<0 0 0.1189 -0.9929>}
 pinky_metacarpal_r {parent=hand_r, loc=<-1.7616 0.9753 -3.3814>, rot=<-0.0397 0.0814 -0.069 -0.9935>}
 pinky_01_r {parent=pinky_metacarpal_r, loc=<-7.1709 0.4831 0.1094>, rot=<-0.0654 -0.0035 0.2242 -0.9723>}
 pinky_02_r {parent=pinky_01_r, loc=<-2.8274 -0.0893 0.1043>, rot=<0.0037 0.0011 0.2854 -0.9584>}
 pinky_03_r {parent=pinky_02_r, loc=<-2.2046 -0.0307 0.0078>, rot=<-0.0002 0.0098 0.1445 -0.9895>}
 hand_attach_r {parent=hand_r, loc=<-8.5001 2.9999 0>, rot=<0 0 -0.7071 -0.7071>}
 deform_wrist_dn_base_r {parent=hand_r, loc=<-0.0004 2.2345 -0.0001>, rot=<0 0 0 -1>}
 deform_wrist_dn_r {parent=deform_wrist_dn_base_r, loc=<0 0 0>, rot=<0 0 0 1>}
 deform_wrist_up_base_r {parent=hand_r, loc=<0.0003 -2.2684 0.0001>, rot=<0 0 0 -1>}
 deform_wrist_up_r {parent=deform_wrist_up_base_r, loc=<0 0 0>, rot=<0 0 0 1>}
 lowerarm_twist_02_r {parent=lowerarm_r, loc=<-9.6199 0 0>, rot=<0 0 0 -1>}
 base_gloves_twist_r {parent=lowerarm_twist_02_r, loc=<0 0 0>, rot=<0 0 0 1>}
 dyn_gloves_twist_r {parent=base_gloves_twist_r, loc=<0 0 0>, rot=<0 0 0 1>}
 lowerarm_twist_01_r {parent=lowerarm_r, loc=<-17.8549 0 0>, rot=<0 0 0 -1>}
 deform_elbow_in_base_r {parent=lowerarm_r, loc=<-0.2111 -5.2591 -0.8275>, rot=<0 0 0 -1>}
 deform_elbow_in_r {parent=deform_elbow_in_base_r, loc=<-0.0001 0 0.0002>, rot=<0 0 0 1>}
 deform_elbow_out_base_r {parent=lowerarm_r, loc=<0 4 -0.0001>, rot=<0 0 0 -1>}
 deform_elbow_out_r {parent=deform_elbow_out_base_r, loc=<0 0 0>, rot=<0 0 0 1>}
 upperarm_twist_01_r {parent=upperarm_r, loc=<-0.0041 0 0>, rot=<0 0 0 -1>}
 deform_upperarm_bk_base_r {parent=upperarm_twist_01_r, loc=<-1.0814 7.2192 1.9455>, rot=<0 0 0 1>}
 deform_upperarm_bk_r {parent=deform_upperarm_bk_base_r, loc=<0 0 0>, rot=<0 0 0 1>}
 deform_upperarm_fr_base_r {parent=upperarm_twist_01_r, loc=<0.3689 -7.0506 2.7652>, rot=<0 0 0 1>}
 deform_upperarm_fr_r {parent=deform_upperarm_fr_base_r, loc=<0 0 0>, rot=<0 0 0 1>}
 deform_upperarm_up_base_r {parent=upperarm_twist_01_r, loc=<3.2214 1.1391 -3.3308>, rot=<0 0 0 1>}
 deform_upperarm_up_r {parent=deform_upperarm_up_base_r, loc=<0 0 0>, rot=<0 0 0 -1>}
 upperarm_twist_02_r {parent=upperarm_r, loc=<-13.7941 0 0>, rot=<0 0 0 -1>}
 base_neck {parent=spine_05, loc=<15.5105 1.9434 0>, rot=<0 0 -0.1506 -0.9886>}
 base_neck_l {parent=base_neck, loc=<0 0 0>, rot=<0 0 0 1>}
 dyn_neck_03_dw_l {parent=base_neck_l, loc=<-3.7701 -11.8134 -12.2254>, rot=<0.164 -0.6878 0.3115 -0.6348>}
 dyn_neck_02_dw_l {parent=base_neck_l, loc=<-1.4546 2.0261 -15.9713>, rot=<0.0417 -0.7132 0.0427 -0.6984>}
 dyn_neck_01_l {parent=base_neck_l, loc=<-3.3326 14.1753 -11.5879>, rot=<-0.096 -0.7294 -0.2233 -0.6394>}
 dyn_neck_02_up_l {parent=base_neck_l, loc=<1.5648 1.344 -15.9713>, rot=<0.0661 -0.7015 -0.0057 -0.7096>}
 dyn_neck_03_up_l {parent=base_neck_l, loc=<4.1833 -11.4381 -12.2254>, rot=<0.3174 -0.732 0.2077 -0.5659>}
 base_neck_acc_l {parent=base_neck_l, loc=<-0.8516 8.5322 -15.7426>, rot=<-0.0438 -0.7058 -0.1908 -0.6809>}
 dyn_neck_acc_l {parent=base_neck_acc_l, loc=<0 0 0>, rot=<0 0 0 1>}
 base_neck_r {parent=base_neck, loc=<0 0 0>, rot=<0 0 0 1>}
 dyn_neck_02_up_r {parent=base_neck_r, loc=<1.5648 1.3441 15.971>, rot=<0.7163 0.0653 0.6945 0.0181>}
 dyn_neck_03_up_r {parent=base_neck_r, loc=<4.0432 -12.0584 10.7361>, rot=<0.6568 -0.296 0.6267 -0.2972>}
 dyn_neck_03_dw_r {parent=base_neck_r, loc=<-3.9103 -12.4338 10.7361>, rot=<0.6489 -0.2809 0.6949 -0.131>}
 dyn_neck_02_dw_r {parent=base_neck_r, loc=<-1.4547 2.0261 15.971>, rot=<0.7164 -0.0035 0.6977 -0.0007>}
 dyn_neck_01_r {parent=base_neck_r, loc=<-3.3326 14.1753 11.588>, rot=<0.6627 0.3036 0.6707 0.137>}
 base_neck_acc_r {parent=base_neck_r, loc=<-0.9147 8.3219 15.3355>, rot=<0.673 0.2169 0.7035 0.0709>}
 dyn_neck_acc_r {parent=base_neck_acc_r, loc=<0 0 0>, rot=<0 0 0 1>}
 base_neck_fr {parent=base_neck, loc=<0 0 0>, rot=<0 0 0 1>}
 dyn_neck_fr {parent=base_neck_fr, loc=<-5.6652 18.8195 0>, rot=<-0.5523 -0.4415 -0.4415 -0.5523>}
 base_neck_bck {parent=base_neck, loc=<0 0 0>, rot=<0 0 0 1>}
 base_neck_bck_up_l {parent=base_neck_bck, loc=<3.9002 -16.4174 0>, rot=<-0.0737 -0.7033 0.7033 0.0737>}
 dyn_neck_bck_up_l {parent=base_neck_bck_up_l, loc=<0 0 0>, rot=<0 0.7024 0 -0.7118>}
 aim_neck_bck_up_l {parent=spine_05, loc=<1.786 -12.1179 0>, rot=<0.0331 -0.7063 0.7063 -0.0331>}
 aim_neck_bck_up_upvector_l {parent=spine_05, loc=<1.786 -12.118 -9.1293>, rot=<0.0331 -0.7063 0.7063 -0.0331>}
 base_neck_bck_center_l {parent=spine_05, loc=<12.8939 -12.2445 0>, rot=<0.0302 -0.7065 0.7065 -0.0302>}
 dyn_neck_bck_center_l {parent=base_neck_bck_center_l, loc=<0 0 0>, rot=<0 0 0 1>}
 deform_pec_aim_base_l {parent=spine_04, loc=<7.8732 12.6388 -2.4123>, rot=<0.0102 -0.4602 0.3979 -0.7936>}
 deform_pec_aim_l {parent=deform_pec_aim_base_l, loc=<0 0 0>, rot=<0 0 0 1>}
 deform_pec_base_l {parent=deform_pec_aim_l, loc=<10.766 0 0>, rot=<0 0 0 1>}
 deform_pec_l {parent=deform_pec_base_l, loc=<0.055 0 0>, rot=<0 0 0 1>}
 deform_pec_aim_base_r {parent=spine_04, loc=<7.8727 12.6388 2.4123>, rot=<-0.4602 -0.0102 0.7936 0.3979>}
 deform_pec_aim_r {parent=deform_pec_aim_base_r, loc=<0 0 0>, rot=<0 0 0 1>}
 deform_pec_base_r {parent=deform_pec_aim_r, loc=<-10.7662 -0.0006 -0.0007>, rot=<0 0 0 1>}
 deform_pec_r {parent=deform_pec_base_r, loc=<0 0 0>, rot=<0 0 0 1>}
 base_belt {parent=spine_02, loc=<0 0 0>, rot=<0 0 0 1>}
 base_dyn_acc_l {parent=base_belt, loc=<5.8728 3.9051 -16.7932>, rot=<0.0282 -0.7065 -0.0282 -0.7065>}
 base_acc_01_l {parent=base_dyn_acc_l, loc=<-1.7558 3.5128 -8.3098>, rot=<0.1623 -0.6882 -0.1623 -0.6882>}
 dyn_acc_01_l {parent=base_acc_01_l, loc=<0 0 0>, rot=<0 0 0 1>}
 base_acc_02_l {parent=base_dyn_acc_l, loc=<-0.2985 -3.3499 -8.0644>, rot=<-0.0188 -0.7069 0.0188 -0.7069>}
 dyn_acc_02_l {parent=base_acc_02_l, loc=<0 0 0>, rot=<0 0 0 1>}
 base_dyn_acc_r {parent=base_belt, loc=<6.7161 4.3321 16.9504>, rot=<0.0282 -0.7065 -0.0282 -0.7065>}
 base_acc_01_r {parent=base_dyn_acc_r, loc=<1.4362 2.3602 -9.4458>, rot=<0.6917 -0.1469 -0.6917 -0.1469>}
 dyn_acc_01_r {parent=base_acc_01_r, loc=<0 0 0>, rot=<0 0 0 1>}
 FX_Idle_01 {parent=dyn_acc_01_r, type=Null, loc=<0 0 0>, rot=<0 0 0 1>}
 base_acc_02_r {parent=base_dyn_acc_r, loc=<0.7727 -4.2188 -8.5127>, rot=<0.706 0.0402 -0.706 0.0402>}
 dyn_acc_02_r {parent=base_acc_02_r, loc=<0 0 0>, rot=<0 0 0 1>}
 FX_Idle_02 {parent=dyn_acc_02_r, type=Null, loc=<0 0 0>, rot=<0 0 0 1>}
 thigh_l {parent=pelvis, loc=<-2.5693 -0.0109 -7.7501>, rot=<0.0012 0.0572 0.0155 -0.9982>}
 calf_l {parent=thigh_l, loc=<-39.0484 0 0>, rot=<0.0006 -0.0161 -0.0391 -0.9991>}
 foot_l {parent=calf_l, loc=<-42.561 0.1077 0.4292>, rot=<0.0418 -0.0455 0.0343 -0.9975>}
 ball_l {parent=foot_l, loc=<-6.1599 13.8264 0.6806>, rot=<0.0001 -0.0001 -0.7063 -0.7079>}
 calf_twist_02_l {parent=calf_l, loc=<-18.8064 -0.0967 0.0397>, rot=<0.0237 0.0095 -0.0035 -0.9997>}
 base_boot_twist_l {parent=calf_twist_02_l, loc=<0 0 0>, rot=<0 0 0 1>}
 dyn_boot_twist_l {parent=base_boot_twist_l, loc=<-23.7409 0.3425 0.8556>, rot=<0 0 0 1>}
 calf_twist_01_l {parent=calf_l, loc=<-35.813 -0.0967 0.0397>, rot=<0.0237 0.0095 -0.0035 -0.9997>}
 deform_knee_in_base_l {parent=calf_l, loc=<1.298 -5.2957 -0.0496>, rot=<0 0 0 -1>}
 deform_knee_in_l {parent=deform_knee_in_base_l, loc=<0 0 0>, rot=<0 0 0 1>}
 deform_knee_out_base_l {parent=calf_l, loc=<0 4 0>, rot=<0 0 0 -1>}
 deform_knee_out_l {parent=deform_knee_out_base_l, loc=<0 0 0>, rot=<0 0 0 1>}
 thigh_twist_01_l {parent=thigh_l, loc=<-21.6303 0 0>, rot=<0.0013 0 0 -1>}
 deform_groin_base_l {parent=thigh_l, loc=<-2.1908 8.1142 -1.7442>, rot=<0 0 0 -1>}
 deform_groin_l {parent=deform_groin_base_l, loc=<0 0 0>, rot=<0 0 0 1>}
 helper_thigh_l {parent=thigh_l, loc=<0 0 0>, rot=<0 0 0 1>}
 thigh_r {parent=pelvis, loc=<-2.5693 -0.0109 7.7501>, rot=<0.0572 -0.0012 0.9982 0.0155>}
 calf_r {parent=thigh_r, loc=<39.0484 0 0>, rot=<0.0006 -0.0161 -0.0391 -0.9991>}
 foot_r {parent=calf_r, loc=<42.561 -0.1077 -0.4292>, rot=<0.0418 -0.0455 0.0343 -0.9975>}
 ball_r {parent=foot_r, loc=<6.1599 -13.8264 -0.6806>, rot=<0.0001 -0.0001 -0.7063 -0.7079>}
 calf_twist_02_r {parent=calf_r, loc=<18.8064 0.0967 -0.0397>, rot=<0.0237 0.0095 -0.0035 -0.9997>}
 base_boot_twist_r {parent=calf_twist_02_r, loc=<0 0 0>, rot=<0 0 1 0>}
 dyn_boot_twist_r {parent=base_boot_twist_r, loc=<-23.7409 0.3425 -0.8556>, rot=<0 0 0 1>}
 calf_twist_01_r {parent=calf_r, loc=<35.813 0.0967 -0.0397>, rot=<0.0237 0.0095 -0.0035 -0.9997>}
 deform_knee_in_base_r {parent=calf_r, loc=<-1.298 5.2957 0.0496>, rot=<0 0 0 -1>}
 deform_knee_in_r {parent=deform_knee_in_base_r, loc=<0 0 0>, rot=<0 0 0 1>}
 deform_knee_out_base_r {parent=calf_r, loc=<0 -4 0>, rot=<0 0 0 -1>}
 deform_knee_out_r {parent=deform_knee_out_base_r, loc=<0 0 0>, rot=<0 0 0 1>}
 thigh_twist_01_r {parent=thigh_r, loc=<21.6303 0 0>, rot=<0.0013 0 0 -1>}
 deform_groin_base_r {parent=thigh_r, loc=<2.1908 -8.1142 1.7442>, rot=<0 0 0 -1>}
 deform_groin_r {parent=deform_groin_base_r, loc=<0 0 0>, rot=<0 0 0 1>}
 deform_glute_base_l {parent=pelvis, loc=<-8.8064 -8.8222 -8.4287>, rot=<0.0275 0.0501 -0.4472 -0.8926>}
 deform_glute_l {parent=deform_glute_base_l, loc=<0 0 0>, rot=<0 0 0 1>}
 deform_glute_base_r {parent=pelvis, loc=<-8.8062 -8.8222 8.4287>, rot=<0.0501 -0.0275 0.8926 -0.4472>}
 deform_glute_r {parent=deform_glute_base_r, loc=<0 0 0>, rot=<0 0 0 1>}
 attach {parent=root, loc=<0 0 0>, rot=<0 0 0 1>}
 ik_hand_root {parent=root, loc=<0 0 0>, rot=<0 0 0 1>}
 ik_hand_gun {parent=ik_hand_root, loc=<-51.9727 4.0883 102.6935>, rot=<0.6084 0.1793 0.4613 -0.6205>}
 ik_hand_r {parent=ik_hand_gun, loc=<0 0 0>, rot=<0 0 0 1>}
 ik_hand_l {parent=ik_hand_gun, loc=<53.0285 82.1781 35.2051>, rot=<-0.5102 -0.7906 -0.3387 0>}
 ik_foot_root {parent=root, loc=<0 0 0>, rot=<0 0 0 1>}
 ik_foot_l {parent=ik_foot_root, loc=<15.2608 -4.4176 9.4432>, rot=<0.0357 0.7045 0.0284 -0.7082>}
 ik_foot_r {parent=ik_foot_root, loc=<-15.2608 -4.4176 9.4432>, rot=<0.7082 0.0284 -0.7045 0.0357>}
 root_fp {parent=root, loc=<0 0 0>, rot=<0 0 0 1>}
 attach_fp {parent=root_fp, loc=<0 0 0>, rot=<0 0 0 1>}
 camera_root {parent=root_fp, loc=<0 0 0>, rot=<0 0 0 1>}
 camera {parent=camera_root, loc=<0 0 0>, rot=<0 0 0 1>}
 attach_emote_01 {parent=root, loc=<0 0 -4.3002>, rot=<0 0 0 1>}
 attach_emote_02 {parent=root, loc=<0 0 -4.3002>, rot=<0 0 0 1>}
 attach_emote_03 {parent=root, loc=<0 0 -4.3002>, rot=<0 0 0 1>}
 warp_point_01 {parent=root, loc=<0 0 -4.3002>, rot=<0 0 0 1>}
 warp_point_02 {parent=root, loc=<0 0 -4.3002>, rot=<0 0 0 1>}
 warp_point_03 {parent=root, loc=<0 0 -4.3002>, rot=<0 0 0 1>}
 clavicle_l_ctrl {controlType=EULER_TRANSFORM, type=Control, loc=<0 0 0>, rot=<0 0 0 1>}
 neck_01_ctrl {controlType=EULER_TRANSFORM, type=Control, loc=<4.1271 -4.3018 142.4487>, rot=<-0.1488 -0.6913 0.0784 0.7027>}
 neck_02_ctrl {controlType=EULER_TRANSFORM, parent=neck_01_ctrl, type=Control, loc=<0 0 0>, rot=<0 0 0 1>}
 head_ctrl {controlType=EULER_TRANSFORM, parent=neck_02_ctrl, type=Control, loc=<-9.0145 -4.3111 152.0342>, rot=<0.1195 -0.6969 0.1137 0.6979>}
]
variables [
 CrouchWalk {type=double, default=0}
 IsParachute {type=bool, default=False}
 IsSliding {type=double, default=0}
]
