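# Animation Blueprint for the GrimeHold male body cosmetic, adding secondary motion to accessory, neck, glove and boot bones.
# The parent mesh pose passes through a Control Rig, then chained AnimDynamics springs, with weighted constraints positioning neck accessories before simulation.
# Each update computes a gravity override and sets sim-space angular alpha by front-end state.
# A back-neck spring is gated by a cape/backbling flag, whose setter chain has no event connected.
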AnimBlueprint /BRCosmetics/Characters/Player/Male/Medium/Bodies/M_MED_GrimeHold/Meshes/M_MED_GrimeHold_AnimBP
target_skeleton=&/M_MED_GrimeHold_Skeleton
graphs [
 AnimGraph {nodes [
   Output Pose(Result=Use cached pose 'boots_cache'()) {layerGroup=None}
   n1 = Get Sim Space Settings Sim Space Angular Alpha() -> (Sim Space Settings Sim Space Angular Alpha:float)
   n2 = Anim Dynamics - Bone: dyn_acc_01_l!(
    ComponentPose=Local To Component(LocalPose=Use cached pose 'controlrig_cache'())
    Alpha=1
    GravityOverride:Vector="0,0,-980"
    SimSpaceSettings_SimSpaceAngularAlpha:float=@n1
    SimSpaceSettings_MaxAngularVelocity:float=10000
    SimSpaceSettings_MaxAngularAcceleration:float=10000
    SimSpaceSettings_ExternalAngularVelocity:Vector="0,0,0"
    ExternalForce="0,0,0"
   ) -> (Pose) {
    angularDampingOverride=1
    angularSpringConstant=50
    bAngularSpring=true
    bOverrideAngularDamping=true
    bUseGravityOverride=true
    gravityScale=0.35
    simulationSpace=RootRelative
    boundBone {boneName=dyn_acc_01_l}
    componentAppliedLinearAccClamp=<100 100 100>
    componentLinearAccScale=<0.3 0.3 0.3>
    componentLinearVelScale=<0.4 0.4 0.4>
    gravityOverride {z=-980}
    physicsBodyDefinitions [{
      boundBone {boneName=dyn_acc_01_l}
      boxExtents=<10 20 20>
      constraintSetup {angularLimitsMax=<0 5 0>, angularLimitsMin=<0 -5 -10>, angularTarget=<1 0 0>}
      localJointOffset=<4 0 0>
     }]
    simSpaceSettings {maxAngularAcceleration=35, maxAngularVelocity=7, simSpaceAngularAlpha=1, externalAngularVelocity=<0.2 0.2 0.2>}
   }
   n3 = Anim Dynamics - Bone: dyn_acc_02_l!(
    ComponentPose=@n2
    Alpha=1
    GravityOverride:Vector="0,0,-980"
    SimSpaceSettings_SimSpaceAngularAlpha:float=@n1
    SimSpaceSettings_MaxAngularVelocity:float=10000
    SimSpaceSettings_MaxAngularAcceleration:float=10000
    SimSpaceSettings_ExternalAngularVelocity:Vector="0,0,0"
    ExternalForce="0,0,0"
   ) -> (Pose) {
    angularDampingOverride=1
    angularSpringConstant=50
    bAngularSpring=true
    bOverrideAngularDamping=true
    bUseGravityOverride=true
    gravityScale=0.35
    simulationSpace=RootRelative
    boundBone {boneName=dyn_acc_02_l}
    componentAppliedLinearAccClamp=<100 100 100>
    componentLinearAccScale=<0.3 0.3 0.3>
    componentLinearVelScale=<0.4 0.4 0.4>
    gravityOverride {z=-980}
    physicsBodyDefinitions [{
      boundBone {boneName=dyn_acc_02_l}
      boxExtents=<10 20 20>
      constraintSetup {angularLimitsMax=<0 5 5>, angularLimitsMin=<0 -5 -5>, angularTarget=<1 0 0>}
      localJointOffset=<4 0 0>
     }]
    simSpaceSettings {maxAngularAcceleration=35, maxAngularVelocity=7, simSpaceAngularAlpha=1, externalAngularVelocity=<0.2 0.2 0.2>}
   }
   n4 = Anim Dynamics - Bone: dyn_acc_01_r!(
    ComponentPose=@n3
    Alpha=1
    GravityOverride:Vector="0,0,-980"
    SimSpaceSettings_SimSpaceAngularAlpha:float=@n1
    SimSpaceSettings_MaxAngularVelocity:float=10000
    SimSpaceSettings_MaxAngularAcceleration:float=10000
    SimSpaceSettings_ExternalAngularVelocity:Vector="0,0,0"
    ExternalForce="0,0,0"
   ) -> (Pose) {
    angularDampingOverride=1
    angularSpringConstant=50
    bAngularSpring=true
    bOverrideAngularDamping=true
    bUseGravityOverride=true
    gravityScale=0.35
    simulationSpace=RootRelative
    boundBone {boneName=dyn_acc_01_r}
    componentAppliedLinearAccClamp=<100 100 100>
    componentLinearAccScale=<0.3 0.3 0.3>
    componentLinearVelScale=<0.4 0.4 0.4>
    gravityOverride {z=-980}
    physicsBodyDefinitions [{
      boundBone {boneName=dyn_acc_01_r}
      boxExtents=<10 10 10>
      constraintSetup {angularLimitsMax=<0 5 5>, angularLimitsMin=<0 -5 -5>, angularTarget=<1 0 0>}
      localJointOffset=<2 0 0>
     }]
    simSpaceSettings {maxAngularAcceleration=35, maxAngularVelocity=7, simSpaceAngularAlpha=1, externalAngularVelocity=<0.2 0.2 0.2>}
   }
   n5 = Anim Dynamics - Bone: dyn_acc_02_r!(
    ComponentPose=@n4
    Alpha=1
    GravityOverride:Vector="0,0,-980"
    SimSpaceSettings_SimSpaceAngularAlpha:float=@n1
    SimSpaceSettings_MaxAngularVelocity:float=10000
    SimSpaceSettings_MaxAngularAcceleration:float=10000
    SimSpaceSettings_ExternalAngularVelocity:Vector="0,0,0"
    ExternalForce="0,0,0"
   ) -> (Pose) {
    angularDampingOverride=1
    angularSpringConstant=50
    bAngularSpring=true
    bOverrideAngularDamping=true
    bUseGravityOverride=true
    gravityScale=0.35
    simulationSpace=RootRelative
    boundBone {boneName=dyn_acc_02_r}
    componentAppliedLinearAccClamp=<100 100 100>
    componentLinearAccScale=<0.3 0.3 0.3>
    componentLinearVelScale=<0.4 0.4 0.4>
    gravityOverride {z=-980}
    physicsBodyDefinitions [{
      boundBone {boneName=dyn_acc_02_r}
      boxExtents=<10 10 10>
      constraintSetup {angularLimitsMax=<2 5 10>, angularLimitsMin=<-3 0 0>, angularTarget=<1 0 0>}
      localJointOffset=<2 0 0>
     }]
    simSpaceSettings {maxAngularAcceleration=35, maxAngularVelocity=7, simSpaceAngularAlpha=1, externalAngularVelocity=<0.2 0.2 0.2>}
   }
   Save cached pose 'belt_acc_cached'(Pose=Component To Local(ComponentPose=@n5)) {cacheName=belt_acc_cached}
   n6 = "Constraint - dyn_neck_acc_l (PPPPPPPPPPP)"!(
    ComponentPose=Local To Component(LocalPose=Use cached pose 'belt_acc_cached'())
    Alpha=1
    ConstraintWeights_0=0.4965
    ConstraintWeights_1:float=0.3253
    ConstraintWeights_2:float=0.0548
    ConstraintWeights_3:float=0.0417
    ConstraintWeights_4:float=0.029
    ConstraintWeights_5:float=0.0215
    ConstraintWeights_6:float=0.0155
    ConstraintWeights_7:float=0.0068
    ConstraintWeights_8:float=0.0034
    ConstraintWeights_9:float=0.0028
    ConstraintWeights_10:float=0.0027
   ) -> (Pose) {
    boneToModify {boneName=dyn_neck_acc_l}
    constraintSetup [
     {targetBone {boneName=dyn_neck_02_up_l}}
     {targetBone {boneName=dyn_neck_01_l}}
     {targetBone {boneName=spine_04}}
     {targetBone {boneName=dyn_neck_02_dw_l}}
     {targetBone {boneName=upperarm_twist_01_l}}
     {targetBone {boneName=clavicle_l}}
     {targetBone {boneName=deform_upperarm_up_l}}
     {targetBone {boneName=dyn_neck_fr}}
     {targetBone {boneName=dyn_neck_03_dw_l}}
     {targetBone {boneName=spine_05}}
     {targetBone {boneName=dyn_neck_03_up_l}}
    ]
    constraintWeights [1,1,1,1,1,1,1,1,1,1,1]
   }
   n7 = "Constraint - dyn_neck_acc_r (PPPPPPPPPPP)"!(
    ComponentPose=@n6
    Alpha=1
    ConstraintWeights_0=0.5407
    ConstraintWeights_1:float=0.2907
    ConstraintWeights_2:float=0.0558
    ConstraintWeights_3:float=0.0305
    ConstraintWeights_4:float=0.0223
    ConstraintWeights_5:float=0.0223
    ConstraintWeights_6:float=0.0176
    ConstraintWeights_7:float=0.0136
    ConstraintWeights_8:float=0.0033
    ConstraintWeights_9:float=0.0031
    ConstraintWeights_10:float=0.0001
   ) -> (Pose) {
    boneToModify {boneName=dyn_neck_acc_r}
    constraintSetup [
     {targetBone {boneName=dyn_neck_01_r}}
     {targetBone {boneName=dyn_neck_02_up_r}}
     {targetBone {boneName=dyn_neck_02_dw_r}}
     {targetBone {boneName=upperarm_twist_01_r}}
     {targetBone {boneName=clavicle_r}}
     {targetBone {boneName=spine_04}}
     {targetBone {boneName=deform_upperarm_up_r}}
     {targetBone {boneName=dyn_neck_fr}}
     {targetBone {boneName=dyn_neck_03_dw_r}}
     {targetBone {boneName=dyn_neck_03_up_r}}
     {targetBone {boneName=spine_05}}
    ]
    constraintWeights [1,1,1,1,1,1,1,1,1,1,1]
   }
   n8 = Get Sim Space Settings Sim Space Angular Alpha()
   n9 = Anim Dynamics - Bone: dyn_neck_acc_l!(
    ComponentPose=@n7
    Alpha=1
    GravityOverride:Vector="0,0,-980"
    SimSpaceSettings_SimSpaceAngularAlpha:float=@n8
    SimSpaceSettings_MaxAngularVelocity:float=10000
    SimSpaceSettings_MaxAngularAcceleration:float=10000
    SimSpaceSettings_ExternalAngularVelocity:Vector="0,0,0"
    ExternalForce="0,0,0"
   ) -> (Pose) {
    angularDampingOverride=1
    angularSpringConstant=50
    bAngularSpring=true
    bOverrideAngularDamping=true
    bUseGravityOverride=true
    gravityScale=0.35
    simulationSpace=RootRelative
    boundBone {boneName=dyn_neck_acc_l}
    componentAppliedLinearAccClamp=<3000 3000 3000>
    componentLinearAccScale=<0.3 0.3 0.3>
    componentLinearVelScale=<0.1 0.1 0.1>
    gravityOverride {z=-980}
    physicsBodyDefinitions [{
      boundBone {boneName=dyn_neck_acc_l}
      boxExtents=<10 10 10>
      constraintSetup {angularLimitsMax=<0 20 0>, angularLimitsMin=<0 -20 0>, angularTarget=<1 0 0>}
      localJointOffset=<4 0 0>
     }]
    simSpaceSettings {maxAngularAcceleration=40, maxAngularVelocity=8, simSpaceAngularAlpha=1, externalAngularVelocity=<0.3 0.3 0.3>}
   }
   n10 = Anim Dynamics - Bone: dyn_neck_acc_r!(
    ComponentPose=@n9
    Alpha=1
    GravityOverride:Vector="0,0,-980"
    SimSpaceSettings_SimSpaceAngularAlpha:float=@n8
    SimSpaceSettings_MaxAngularVelocity:float=10000
    SimSpaceSettings_MaxAngularAcceleration:float=10000
    SimSpaceSettings_ExternalAngularVelocity:Vector="0,0,0"
    ExternalForce="0,0,0"
   ) -> (Pose) {
    angularDampingOverride=1
    angularSpringConstant=50
    bAngularSpring=true
    bOverrideAngularDamping=true
    bUseGravityOverride=true
    gravityScale=0.35
    simulationSpace=RootRelative
    boundBone {boneName=dyn_neck_acc_r}
    componentAppliedLinearAccClamp=<3000 3000 3000>
    componentLinearAccScale=<0.3 0.3 0.3>
    componentLinearVelScale=<0.1 0.1 0.1>
    gravityOverride {z=-980}
    physicsBodyDefinitions [{
      boundBone {boneName=dyn_neck_acc_r}
      boxExtents=<10 10 10>
      constraintSetup {angularLimitsMax=<0 20 0>, angularLimitsMin=<0 -20 0>, angularTarget=<1 0 0>}
      localJointOffset=<4 0 0>
     }]
    simSpaceSettings {maxAngularAcceleration=40, maxAngularVelocity=8, simSpaceAngularAlpha=1, externalAngularVelocity=<0.3 0.3 0.3>}
   }
   n11 = Anim Dynamics - Bone: dyn_neck_bck_center_l!(ComponentPose=@n10, bAlphaBoolEnabled:bool=Get is Wearing cape and backpack() -> (is Wearing cape and backpack:bool)) -> (Pose) {
    alphaInputType=Bool
    angularDampingOverride=1
    angularSpringConstant=3000
    bAngularSpring=true
    bOverrideAngularDamping=true
    bUseGravityOverride=true
    gravityScale=0.35
    simulationSpace=BoneRelative
    alphaBoolBlend {blendOption=Linear}
    boundBone {boneName=dyn_neck_bck_center_l}
    componentAppliedLinearAccClamp=<3000 3000 3000>
    componentLinearAccScale=<0.52 0.52 0.52>
    componentLinearVelScale=<0.52 0.52 0.52>
    gravityOverride {z=-980}
    physicsBodyDefinitions [{
      boundBone {boneName=dyn_neck_bck_center_l}
      boxExtents=<20 40 40>
      constraintSetup {angularLimitsMax=<0 0 20>, angularLimitsMin=<0 -20 -20>, angularTarget=<1 0 0>}
      localJointOffset=<0 0 0>
     }]
    simSpaceSettings {maxAngularAcceleration=35, maxAngularVelocity=7, simSpaceAngularAlpha=0.35}
   }
   n12 = Get Sim Space Settings Sim Space Angular Alpha()
   n13 = Anim Dynamics - Bone: dyn_gloves_twist_r!(
    ComponentPose=@n11
    Alpha=1
    GravityOverride:Vector="0,0,-980"
    SimSpaceSettings_SimSpaceAngularAlpha:float=@n12
    SimSpaceSettings_MaxAngularVelocity:float=10000
    SimSpaceSettings_MaxAngularAcceleration:float=10000
    SimSpaceSettings_ExternalAngularVelocity:Vector="0,0,0"
    ExternalForce="0,0,0"
   ) -> (Pose) {
    angularDampingOverride=1
    angularSpringConstant=3000
    bAngularSpring=true
    bOverrideAngularDamping=true
    bUseGravityOverride=true
    gravityScale=0.35
    simulationSpace=RootRelative
    boundBone {boneName=dyn_gloves_twist_r}
    componentAppliedLinearAccClamp=<3000 3000 3000>
    componentLinearAccScale=<0.3 0.3 0.3>
    componentLinearVelScale=<0.3 0.3 0.3>
    gravityOverride {z=-980}
    physicsBodyDefinitions [{
      boundBone {boneName=dyn_gloves_twist_r}
      boxExtents=<25 25 25>
      constraintSetup {angularLimitsMax=<10 10 10>, angularLimitsMin=<-10 -10 -10>, angularTarget=<1 0 0>}
      localJointOffset=<4 0 0>
     }]
    simSpaceSettings {maxAngularAcceleration=40, maxAngularVelocity=8, simSpaceAngularAlpha=1, externalAngularVelocity=<0.3 0.3 0.3>}
   }
   n14 = Anim Dynamics - Bone: dyn_gloves_twist_l!(
    ComponentPose=@n13
    Alpha=1
    GravityOverride:Vector="0,0,-980"
    SimSpaceSettings_SimSpaceAngularAlpha:float=@n12
    SimSpaceSettings_MaxAngularVelocity:float=10000
    SimSpaceSettings_MaxAngularAcceleration:float=10000
    SimSpaceSettings_ExternalAngularVelocity:Vector="0,0,0"
    ExternalForce="0,0,0"
   ) -> (Pose) {
    angularDampingOverride=1
    angularSpringConstant=3000
    bAngularSpring=true
    bOverrideAngularDamping=true
    bUseGravityOverride=true
    gravityScale=0.35
    simulationSpace=RootRelative
    boundBone {boneName=dyn_gloves_twist_l}
    gravityOverride {z=-980}
    physicsBodyDefinitions [{
      boundBone {boneName=dyn_gloves_twist_l}
      boxExtents=<25 25 25>
      constraintSetup {angularLimitsMax=<10 10 10>, angularLimitsMin=<-10 -10 -10>, angularTarget=<1 0 0>}
      localJointOffset=<4 0 0>
     }]
    simSpaceSettings {maxAngularAcceleration=40, maxAngularVelocity=8, externalAngularVelocity=<0.3 0.3 0.3>}
   }
   n15 = Get Sim Space Settings Sim Space Angular Alpha()
   n16 = Anim Dynamics - Bone: dyn_boot_twist_r!(
    ComponentPose=@n14
    Alpha=1
    GravityOverride:Vector="0,0,-980"
    SimSpaceSettings_SimSpaceAngularAlpha:float=@n15
    SimSpaceSettings_MaxAngularVelocity:float=10000
    SimSpaceSettings_MaxAngularAcceleration:float=10000
    SimSpaceSettings_ExternalAngularVelocity:Vector="0,0,0"
    ExternalForce="0,0,0"
   ) -> (Pose) {
    angularDampingOverride=1
    angularSpringConstant=5000
    bAngularSpring=true
    bOverrideAngularDamping=true
    bUseGravityOverride=true
    gravityScale=0.35
    simulationSpace=RootRelative
    boundBone {boneName=dyn_boot_twist_r}
    componentAppliedLinearAccClamp=<3000 3000 3000>
    componentLinearAccScale=<0.3 0.3 0.3>
    componentLinearVelScale=<0.3 0.3 0.3>
    gravityOverride {z=-980}
    physicsBodyDefinitions [{
      boundBone {boneName=dyn_boot_twist_r}
      boxExtents=<15 15 15>
      constraintSetup {angularLimitsMax=<3 3 3>, angularLimitsMin=<-3 -3 -3>, angularTarget=<1 0 0>}
      localJointOffset=<4 0 0>
     }]
    simSpaceSettings {maxAngularAcceleration=40, maxAngularVelocity=8, simSpaceAngularAlpha=1, externalAngularVelocity=<0.3 0.3 0.3>}
   }
   n17 = Anim Dynamics - Bone: dyn_boot_twist_l!(
    ComponentPose=@n16
    Alpha=1
    GravityOverride:Vector="0,0,-980"
    SimSpaceSettings_SimSpaceAngularAlpha:float=@n15
    SimSpaceSettings_MaxAngularVelocity:float=8
    SimSpaceSettings_MaxAngularAcceleration:float=40
    SimSpaceSettings_ExternalAngularVelocity:Vector="0.3,0.3,0.3"
    ExternalForce="0.3,0.3,0.3"
   ) -> (Pose) {
    angularDampingOverride=1
    angularSpringConstant=5000
    bAngularSpring=true
    bOverrideAngularDamping=true
    bUseGravityOverride=true
    gravityScale=0.35
    simulationSpace=RootRelative
    boundBone {boneName=dyn_boot_twist_l}
    componentAppliedLinearAccClamp=<3000 3000 3000>
    componentLinearAccScale=<0.3 0.3 0.3>
    componentLinearVelScale=<0.3 0.3 0.3>
    gravityOverride {z=-980}
    physicsBodyDefinitions [{
      boundBone {boneName=dyn_boot_twist_l}
      boxExtents=<15 15 15>
      constraintSetup {angularLimitsMax=<3 3 3>, angularLimitsMin=<-3 -3 -3>, angularTarget=<1 0 0>}
      localJointOffset=<4 0 0>
     }]
    simSpaceSettings {maxAngularAcceleration=40, maxAngularVelocity=8, externalAngularVelocity=<0.3 0.3 0.3>}
   }
   Save cached pose 'boots_cache'(Pose=Component To Local(ComponentPose=@n17)) {cacheName=boots_cache}
   n18 = Control Rig!(Source=Copy Pose From Mesh() {bUseAttachedParent=true}, Alpha, IsParachute:bool=Get bIsParachuteOpened()) -> (Pose) {
    alphaBoolBlend {blendOption=Linear}
    controlRigAssetReference {blueprintRigClass=&/M_MED_GrimeHold_CtrlRig.M_MED_GrimeHold_CtrlRig_C}
   }
   Save cached pose 'controlrig_cache'(Pose=@n18) {cacheName=controlrig_cache}
  ]}
 EventGraph {nodes [
   Event OnCharacterCustomizationLoaded()
   > n1 = Cast To PlayerPawn_Athena(Object=TryGetPawnOwner())
   > Set Body_Mesh(Body_Mesh=GetSkeletalMeshForPartType(self=@n1))
   n2 = CheckForBackbling()
   > n3 = Set isWearingBackbling(isWearingBackbling=AND Boolean!(A=@n2, B=NOT Boolean(A=@n2.bIsBackblingACape), C:bool=NOT Boolean(A=@n2.bIsBackblingLong)) -> (ReturnValue))
   > n4 = Set Is Wearing Cape(Is Wearing Cape=OR Boolean(A=@n2.bIsBackblingACape, B=@n2.bIsBackblingLong))
   > Set is Wearing cape and backpack(is Wearing cape and backpack:bool=NOT Boolean(A=OR Boolean(A=@n4, B=@n3))) -> (Output_Get:bool)
   Event BlueprintUpdateAnimation()
   > n5 = Sequence()
   > n6 = "Gravity Override State Function (Parameter Asset) (Nativized)"(
    CharacterPartAnimInstance=Self-Reference()
    GameTimeInSeconds=GetGameTimeInSeconds()
    Parameters=&/M_MED_GrimeHold_Parameters
   )
   > Set GravityOverride(GravityOverride=@n6)
   Set accesories(execute=@n5.then_1, accesories:bool=Get accesories() -> (accesories:bool)) -> (Output_Get:bool)
   > n7 = Branch(Condition=Get bIsInFrontEnd())
   > Set Sim Space Settings Sim Space Angular Alpha(Sim Space Settings Sim Space Angular Alpha:float=0.3) -> (Output_Get:float)
   Set Sim Space Settings Sim Space Angular Alpha(execute=@n7.else, Sim Space Settings Sim Space Angular Alpha=0)
   Set!(
    execute
    Sim Space Settings_SimSpaceAngularAlpha:float=0
    Sim Space Settings_MaxAngularVelocity:float=10000
    Sim Space Settings_MaxAngularAcceleration:float=10000
    Sim Space Settings_ExternalAngularVelocity:Vector="0,0,0"
   ) -> (then, Output_Get:Anim Phys Sim Space Settings Structure)
   Set(
    Sim Space Settings_SimSpaceAngularAlpha=1
    Sim Space Settings_MaxAngularVelocity=10000
    Sim Space Settings_MaxAngularAcceleration=10000
    Sim Space Settings_ExternalAngularVelocity="0,0,0"
   )
  ]}
]
variables [
 accesories {type=boolean}
 alpha frontend {type=number}
 Body_Mesh {type=SkeletalMeshComponent}
 Gravity Override {type=Vector}
 Gravity Override Cape {type=Vector}
 GravityOverride {type=Vector}
 Is Wearing Cape {type=boolean}
 is Wearing cape and backpack {type=boolean}
 isWearingBackbling {type=boolean}
 RBAN Params_0 {
  type=FortRigidBodyAnimNodeParameterSet
  default=(Skydive_GoingUp=(RigidBodyAnimNode=()),Skydive_FloatingDown=(RigidBodyAnimNode=()),Skydive_DivingDown=(RigidBodyAnimNode=()),Skydive_Parachuting=(RigidBodyAnimNode=()),OnGround_Standing=(RigidBodyAnimNode=()),OnGround_Moving=(RigidBodyAnimNode=()),OnGround_Moving_HighRotationRate=(RigidBodyAnimNode=()),OnGround_Crouch_Moving=(RigidBodyAnimNode=()),Emote_or_Melee=(RigidBodyAnimNode=()),Swimming=(RigidBodyAnimNode=()),Front_End=(RigidBodyAnimNode=()),DBNO=(RigidBodyAnimNode=()),DBNO_Carried=(RigidBodyAnimNode=()),OnGround_Falling=(RigidBodyAnimNode=()),Ziplining=(RigidBodyAnimNode=()),Skydive_Surfing=(RigidBodyAnimNode=()),Riding_Standing=(RigidBodyAnimNode=()),Riding_Moving=(RigidBodyAnimNode=()),Riding_Moving_Targeting=(RigidBodyAnimNode=()),Tactical_Sprint=(RigidBodyAnimNode=()),Rail_Grinding=(RigidBodyAnimNode=()),)
 }
 RBAN Settings Out {type=FortRigidBodyAnimNodeParameters}
 RBAN Settings Out_0 {type=FortRigidBodyAnimNodeParameters}
 Sim Space Settings {
  type=AnimPhysSimSpaceSettings
  default {maxAngularAcceleration=40, maxAngularVelocity=8, simSpaceAngularAlpha=0, externalAngularVelocity=<0.3 0.3 0.3>}
 }
 Sim Space Settings Sim Space Angular Alpha {type=number, default=1}
 simspace var {type=integer}
]
