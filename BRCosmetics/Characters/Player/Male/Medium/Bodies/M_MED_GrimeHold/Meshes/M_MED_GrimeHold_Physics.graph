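# Physics asset for the GrimeHold medium male player body cosmetic, covering two left-side dynamic accessory bones, apparently for secondary motion.
# Capsule bodies cover the pelvis, left thigh helper, two tiny accessory base bones and two dynamic accessory bones.
# Both bases and the thigh helper have angularly locked constraints to the pelvis.
# Each dynamic bone hangs off its matching base with narrow swing and slightly wider twist limits.
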PhysicsAsset /BRCosmetics/Characters/Player/Male/Medium/Bodies/M_MED_GrimeHold/Meshes/M_MED_GrimeHold_Physics
bodies {
 base_acc_01_l [{shapeType=Capsule, length=0.505, radius=0.505, shapeName=base_acc_01_l_capsule, rotation=<rot 0 -90 0>}]
 base_acc_02_l [{shapeType=Capsule, length=0.505, radius=0.505, shapeName=base_acc_02_l_capsule, rotation=<rot 0 -90 0>}]
 dyn_acc_01_l [{
   shapeType=Capsule
   length=6.4938
   radius=3.3554
   shapeName=dyn_acc_01_l_capsule
   center=<2.6533 -1.4497 -5.7117>
   rotation=<rot 3.4901 -90 -9.252>
  }]
 dyn_acc_02_l [{
   shapeType=Capsule
   length=6.4889
   radius=3.3546
   shapeName=dyn_acc_02_l_capsule
   center=<2.2743 0.1383 -5.7309>
   rotation=<rot -0.0002 -90 -7.0947>
  }]
 helper_thigh_l [{shapeType=Capsule, length=19.0679, radius=12.4263, shapeName=thigh_l_capsule, center=<-7.8054 1.3777 2.0974>, rotation=<rot 90 0 0>}]
 pelvis [{
   shapeType=Capsule
   length=14.1788
   radius=9.6443
   shapeName=pelvis_capsule
   center=<4.5303 0.8115 0.1368>
   rotation=<rot -5.0518 -90 -1.0423>
  }]
}
constraints [
 {
  bone1Name=base_acc_02_l
  bone2Name=pelvis
  swing1LimitDegrees=45
  swing1Motion=Locked
  swing2LimitDegrees=45
  swing2Motion=Locked
  twistLimitDegrees=45
  twistMotion=Locked
 }
 {bone1Name=dyn_acc_02_l, bone2Name=base_acc_02_l, swing1LimitDegrees=3, swing2LimitDegrees=3, twistLimitDegrees=8}
 {bone1Name=dyn_acc_01_l, bone2Name=base_acc_01_l, swing1LimitDegrees=3, swing2LimitDegrees=3, twistLimitDegrees=8}
 {
  bone1Name=base_acc_01_l
  bone2Name=pelvis
  swing1LimitDegrees=45
  swing1Motion=Locked
  swing2LimitDegrees=45
  swing2Motion=Locked
  twistLimitDegrees=45
  twistMotion=Locked
 }
 {
  bone1Name=helper_thigh_l
  bone2Name=pelvis
  swing1LimitDegrees=45
  swing1Motion=Locked
  swing2LimitDegrees=45
  swing2Motion=Locked
  twistLimitDegrees=45
  twistMotion=Locked
 }
]
